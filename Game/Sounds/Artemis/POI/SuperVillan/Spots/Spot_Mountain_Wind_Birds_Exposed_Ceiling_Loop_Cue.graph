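# Looping sound cue for an indoor ambient spot at the SuperVillan POI, suggesting mountain wilderness heard through an exposed ceiling.
# One randomly delayed loop picks either desert wind whistles or grassland rustles layered with alpine one-shots or mountain gusts.
# A second delayed loop plays wolf howls, mixed louder, with modulators varying each layer.
# The quality switch fills two of four slots with this mix, and playback borrows Hollow Base window-wind attenuation.
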SoundCue /Game/Sounds/Artemis/POI/SuperVillan/Spots/Spot_Mountain_Wind_Birds_Exposed_Ceiling_Loop_Cue
root=QualityLevel_1
nodes [
 {id=Delay_0, children [Modulator_6], properties {delayMax=45, delayMin=18}}
 {id=Delay_1, children [Random_6], properties {delayMax=35, delayMin=12}}
 {id=Looping_0, children [Delay_0]}
 {id=Looping_1, children [Delay_1]}
 {id=Mixer_0, children [Modulator_5], properties {inputVolume [1.5]}}
 {id=Mixer_2, children [Modulator_3,Modulator_4], properties {inputVolume [1,1]}}
 {id=Mixer_3, children [Modulator_4,Modulator_1], properties {inputVolume [1,1]}}
 {id=Mixer_4, children [Looping_1,Looping_0], properties {inputVolume [0.6,3]}}
 {id=Modulator_1, children [Random_0], properties {pitchMax=1, pitchMin=0.8, volumeMax=1.1, volumeMin=0.9}}
 {id=Modulator_3, children [Random_2], properties {pitchMax=1.1, pitchMin=0.8, volumeMax=1.1, volumeMin=0.9}}
 {id=Modulator_4, children [Random_4], properties {pitchMax=1, pitchMin=0.8, volumeMax=1.1, volumeMin=0.9}}
 {id=Modulator_5, children [Random_7], properties {volumeMax=2.1, volumeMin=2}}
 {id=Modulator_6, children [Random_8], properties {pitchMax=1.1, volumeMax=1.1, volumeMin=0.9}}
 {id=QualityLevel_1, children [Mixer_4,Mixer_4,null,null]}
 {id=Random_0, children [WavePlayer_0,WavePlayer_1,WavePlayer_2], properties {weights [1,1,1]}}
 {id=Random_2, children [WavePlayer_9, WavePlayer_10, WavePlayer_11, WavePlayer_12], properties {weights [1,1,1,1]}}
 {id=Random_4, children [WavePlayer_17, WavePlayer_18, WavePlayer_19, WavePlayer_21, WavePlayer_22], properties {weights [1,1,1,1,1]}}
 {id=Random_5, children [Mixer_2,Mixer_3], properties {weights [1,1]}}
 {id=Random_6, children [Random_5,Mixer_0], properties {weights [1,1]}}
 {id=Random_7, children [WavePlayer_20, WavePlayer_27, WavePlayer_28, WavePlayer_29, WavePlayer_30], properties {weights [1,1,1,1,1]}}
 {
  id=Random_8
  children [WavePlayer_31, WavePlayer_32, WavePlayer_33, WavePlayer_34, WavePlayer_35, WavePlayer_36]
  properties {weights [1,1,1,1,1,1]}
 }
 {id=WavePlayer_0, properties {soundWaveAssetPtr=/Game/Sounds/Artemis/Biomes/Mountains/Mountain_Spots/Spot_Mountain_Day_Gusty_Wind_06}}
 {id=WavePlayer_1, properties {soundWaveAssetPtr=/Game/Sounds/Artemis/Biomes/Mountains/Mountain_Spots/Spot_Mountain_Day_Gusty_Wind_01}}
 {id=WavePlayer_10, properties {soundWaveAssetPtr=/Game/Sounds/Apollo/Biomes/Alpine/Biome_Alpine_Oneshot_Day_Quad_02}}
 {id=WavePlayer_11, properties {soundWaveAssetPtr=/Game/Sounds/Apollo/Biomes/Alpine/Biome_Alpine_Oneshot_Day_Quad_03}}
 {id=WavePlayer_12, properties {soundWaveAssetPtr=/Game/Sounds/Apollo/Biomes/Alpine/Biome_Alpine_Oneshot_Day_Quad_04}}
 {id=WavePlayer_17, properties {soundWaveAssetPtr=/Game/Sounds/Artemis/Biomes/Grasslands/Grassland_Spots/Spot_Grassland_Wind_Rustle_01}}
 {id=WavePlayer_18, properties {soundWaveAssetPtr=/Game/Sounds/Artemis/Biomes/Grasslands/Grassland_Spots/Spot_Grassland_Wind_Rustle_02}}
 {id=WavePlayer_19, properties {soundWaveAssetPtr=/Game/Sounds/Artemis/Biomes/Grasslands/Grassland_Spots/Spot_Grassland_Wind_Rustle_03}}
 {id=WavePlayer_2, properties {soundWaveAssetPtr=/Game/Sounds/Artemis/Biomes/Mountains/Mountain_Spots/Spot_Mountain_Day_Gusty_Wind_02}}
 {id=WavePlayer_20, properties {soundWaveAssetPtr=/Game/Sounds/Fort_Ambient_NonLoop/Desert/Wind_DesertGust_Whistle14}}
 {id=WavePlayer_21, properties {soundWaveAssetPtr=/Game/Sounds/Artemis/Biomes/Grasslands/Grassland_Spots/Spot_Grassland_Wind_Rustle_05}}
 {id=WavePlayer_22, properties {soundWaveAssetPtr=/Game/Sounds/Artemis/Biomes/Grasslands/Grassland_Spots/Spot_Grassland_Wind_Rustle_06}}
 {id=WavePlayer_27, properties {soundWaveAssetPtr=/Game/Sounds/Fort_Ambient_NonLoop/Desert/Wind_DesertGust_Whistle15}}
 {id=WavePlayer_28, properties {soundWaveAssetPtr=/Game/Sounds/Fort_Ambient_NonLoop/Desert/Wind_DesertGust_Whistle16}}
 {id=WavePlayer_29, properties {soundWaveAssetPtr=/Game/Sounds/Fort_Ambient_NonLoop/Desert/Wind_DesertGust_Whistle18}}
 {id=WavePlayer_30, properties {soundWaveAssetPtr=/Game/Sounds/Fort_Ambient_NonLoop/Desert/Wind_DesertGust_Whistle26}}
 {id=WavePlayer_31, properties {soundWaveAssetPtr=/Game/Sounds/Artemis/Biomes/Mountains/Mountain_Spots/Spot_Mountain_Wolf_Howls_01}}
 {id=WavePlayer_32, properties {soundWaveAssetPtr=/Game/Sounds/Artemis/Biomes/Mountains/Mountain_Spots/Spot_Mountain_Wolf_Howls_02}}
 {id=WavePlayer_33, properties {soundWaveAssetPtr=/Game/Sounds/Artemis/Biomes/Mountains/Mountain_Spots/Spot_Mountain_Wolf_Howls_03}}
 {id=WavePlayer_34, properties {soundWaveAssetPtr=/Game/Sounds/Artemis/Biomes/Mountains/Mountain_Spots/Spot_Mountain_Wolf_Howls_04}}
 {id=WavePlayer_35, properties {soundWaveAssetPtr=/Game/Sounds/Artemis/Biomes/Mountains/Mountain_Spots/Spot_Mountain_Wolf_Howls_05}}
 {id=WavePlayer_36, properties {soundWaveAssetPtr=/Game/Sounds/Artemis/Biomes/Mountains/Mountain_Spots/Spot_Mountain_Wolf_Howls_06}}
 {id=WavePlayer_9, properties {soundWaveAssetPtr=/Game/Sounds/Apollo/Biomes/Alpine/Biome_Alpine_Oneshot_Day_Quad_01}}
]
properties {
 attenuationSettings=/Game/Sounds/Artemis/POI/Hollow_Base/Spots/Spot_Interior_Window_Wind_Att
 bEnableBusSends=false
 maxDistance=1253.7231
 pitchMultiplier=0.8
 soundClassObject=&Fort_Ambient_Indoors_Athena
 volumeMultiplier=0.15
 preEffectBusSends [{}]
}
